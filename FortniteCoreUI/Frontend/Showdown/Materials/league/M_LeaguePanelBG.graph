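# Translucent UI material for the Showdown league panel background, drawing a dark purple-to-blue fill with adjustable, tiltable edges.
# Two exponential radial gradients with parameterized radius and density blend three fixed colors; each center uses its position parameter's red channel for both axes.
# Four SmoothStep edge masks on rotated UVs, sharing an EdgeSoftness parameter, combine through Max and invert into a procedural panel shape without textures.
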Material /FortniteCoreUI/Frontend/Showdown/Materials/league/M_LeaguePanelBG
graph {
 properties {
  bUsedWithStaticMesh=false
  blendMode=BLEND_Translucent
  materialDomain=MD_UI
  pixelDepthOffsetMode=PDOM_Legacy
  usedShadingModels=Default Lit
 }
 nodes [
  n1 = VectorParameter() {group=Gradient1, parameterName=Gradient1Position, defaultValue=<rgba 0 0 0 1>}
  n2 = AppendVector(A=@n1.R, B=@n1.R)
  n3 = ScalarParameter() {defaultValue=0.5, group=Gradient1, parameterName=Gradient1Radius}
  n4 = ScalarParameter() {defaultValue=0.5, group=Gradient1, parameterName=Gradient1Density}
  n5 = VectorParameter() {group=Gradient2, parameterName=Gradient2Position, defaultValue=<rgba 0 0 0 1>}
  LinearInterpolate(
   A=LinearInterpolate(
    A=Constant3Vector() {constant=<rgba 0.0273 0.0003 0.117 1>}
    B=Constant3Vector() {constant=<rgba 0.0976 0.0075 0.0887 1>}
    Alpha=MaterialFunctionCall(CenterPosition=@n2, Radius=@n3, Density=@n4) {materialFunction=&RadialGradientExponential}.RadialGradientExponential
   )
   B=Constant3Vector() {constant=<rgba 0.0232 0.0529 0.3613 0>}
   Alpha=MaterialFunctionCall(
    CenterPosition=AppendVector(A=@n5.R, B=@n5.R)
    Radius=ScalarParameter() {defaultValue=0.5, group=Gradient2, parameterName=Gradient2Radius}
    Density=ScalarParameter() {defaultValue=0.5, group=Gradient2, parameterName=Gradient2Density}
   ) {materialFunction=&RadialGradientExponential}.RadialGradientExponential
  )
  n6 = ScalarParameter() {defaultValue=0.1, parameterName=LeftEdge}
  n7 = Constant() {r=0.5}
  n8 = ComponentMask(None=MaterialFunctionCall(
    UVs=TextureCoordinate()
    Rotation Center=AppendVector(A=@n6, B=@n7)
    "Rotation Angle (0-1)"=ScalarParameter() {parameterName=LeftRotation, sliderMax=0.25, sliderMin=-0.25}
   ) {materialFunction=&CustomRotator}.Rotated Values) {r=true}
  n9 = Add(A=@n6, B=ScalarParameter() {defaultValue=0.001, parameterName=EdgeSoftness})
  n10 = OneMinus(None=ScalarParameter() {defaultValue=0.1, parameterName=RightEdge})
  n11 = MaterialFunctionCall(
   UVs=TextureCoordinate()
   Rotation Center=AppendVector(A=@n10, B=Constant() {r=0.5})
   "Rotation Angle (0-1)"=ScalarParameter() {parameterName=RightRotation, sliderMax=0.25, sliderMin=-0.25}
  ) {materialFunction=&CustomRotator}
  n12 = ScalarParameter() {defaultValue=0.001, parameterName=EdgeSoftness}
  n13 = Max(
   A=OneMinus(None=MaterialFunctionCall(Alpha=@n8, Min=@n6, Max=@n9) {materialFunction=&SmoothStep}.Result)
   B=MaterialFunctionCall(Alpha=ComponentMask(None=@n11.Rotated Values) {r=true}, Min=@n10, Max=Add(A=@n10, B=@n12)) {materialFunction=&SmoothStep}.Result
  )
  n14 = TextureCoordinate()
  n15 = OneMinus(None=ScalarParameter() {defaultValue=0.1, parameterName=BottomEdge})
  n16 = AppendVector(A=@n15, B=Constant() {r=0.5})
  n17 = ScalarParameter() {parameterName=BottomRotation, sliderMax=0.25, sliderMin=-0.25}
  n18 = MaterialFunctionCall(
   Alpha=ComponentMask(None=MaterialFunctionCall(UVs=@n14, Rotation Center=@n16, "Rotation Angle (0-1)"=@n17) {materialFunction=&CustomRotator}.Rotated Values) {g=true}
   Min=@n15
   Max=Add(A=@n15, B=ScalarParameter() {defaultValue=0.001, parameterName=EdgeSoftness})
  ) {materialFunction=&SmoothStep}
  n19 = TextureCoordinate()
  n20 = ScalarParameter() {defaultValue=0.1, parameterName=TopEdge}
  n21 = AppendVector(A=@n20, B=Constant() {r=0.5})
  n22 = ScalarParameter() {parameterName=TopRotation, sliderMax=0.25, sliderMin=-0.25}
  n23 = MaterialFunctionCall(
   Alpha=ComponentMask(None=MaterialFunctionCall(UVs=@n19, Rotation Center=@n21, "Rotation Angle (0-1)"=@n22) {materialFunction=&CustomRotator}.Rotated Values) {g=true}
   Min=@n20
   Max=Add(A=@n20, B=ScalarParameter() {defaultValue=0.001, parameterName=EdgeSoftness})
  ) {materialFunction=&SmoothStep}
  OneMinus(None=Max(A=Max(A=@n13, B=@n18.Result), B=OneMinus(None=@n23.Result)))
 ]
}
