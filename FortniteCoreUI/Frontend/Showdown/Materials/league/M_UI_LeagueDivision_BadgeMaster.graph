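# UI master material for Showdown league division badges that color-dodges a tinted shine over the badge texture.
# The texture's red channel minus GemAmount-scaled blue feeds a bump offset as height, distorting the UVs of MF_Shine for a sense of depth.
# A nine-second time loop drives the shine offset lerp, but both ends read ShineRange's red channel, so the offset stays constant as wired.
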Material /FortniteCoreUI/Frontend/Showdown/Materials/league/M_UI_LeagueDivision_BadgeMaster
graph {
 properties {
  bUsedWithStaticMesh=false
  blendMode=BLEND_Translucent
  materialDomain=MD_UI
  pixelDepthOffsetMode=PDOM_Legacy
  usedShadingModels=Default Lit
 }
 nodes [
  n1 = VectorParameter() {parameterName=ShineColor, defaultValue=<rgba 1 0.9815 0.6995 1>}
  n2 = VectorParameter() {parameterName=ShineRange, defaultValue=<rgba -0.8685 7.8291 0 0>}
  n3 = Time() {bOverride_Period=true, period=9}
  n4 = TextureSampleParameter2D() {parameterName=Texture, texture=&/T_UI_LeagueDivision_FancyIcon1}
  n5 = Subtract(A=@n4.R, B=Multiply(A=@n4.B, B=ScalarParameter() {defaultValue=1, parameterName=GemAmount}))
  n6 = MaterialFunctionCall!(
   ShineOffset=LinearInterpolate(A=@n2.R, B=@n2.R, Alpha=Divide(A=@n3) {constB=9})
   Rotation=ScalarParameter() {defaultValue=0.261, parameterName=ShineRotation}
   RotationCenter
   UVs=MaterialFunctionCall!(
    HeightTexture=Desaturation(None=@n5)
    TextureCoordinates
    ReferencePlane=ScalarParameter() {defaultValue=0.4476, group=ShineBump, parameterName=BumpAngle}
    HeightRatioInput=ScalarParameter() {defaultValue=1, group=ShineBump, parameterName=BumpHeight}
    CameraVector
   ) {materialFunction=/Engine/Functions/Engine_MaterialFunctions02/Utility/BumpOffset_Advanced.BumpOffset_advanced}.UVs
  ) {materialFunction=/Game/Athena/HUD/Materials/Victory/MF_Shine}
  Saturate(None=MaterialFunctionCall!(Blend=LinearInterpolate(B=@n1.RGB, Alpha=@n6.Result), Base=@n4.RGB) {materialFunction=/Engine/Functions/Engine_MaterialFunctions03/Blends/Blend_ColorDodge}.Result)
 ]
}
