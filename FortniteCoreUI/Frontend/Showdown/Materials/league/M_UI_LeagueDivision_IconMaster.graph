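# Translucent UI master material for Showdown league division icons, drawing a tinted shape with an optional outline from one texture.
# SmoothStep thresholds the texture's RGB at Edge, distance-field style, for the fill, and again at an Outline offset for the border.
# The fill mask blends OutlineColor into Color, while the Max of both masks forms the overall shape mask.
# OutlineAmount scales both the outline's tint and its mask.
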Material /FortniteCoreUI/Frontend/Showdown/Materials/league/M_UI_LeagueDivision_IconMaster
graph {
 properties {
  bUsedWithStaticMesh=false
  blendMode=BLEND_Translucent
  materialDomain=MD_UI
  pixelDepthOffsetMode=PDOM_Legacy
  usedShadingModels=Default Lit
 }
 nodes [
  n1 = VectorParameter() {parameterName=Color, defaultValue=<rgba 1 1 1 1>}
  n2 = ScalarParameter() {parameterName=OutlineAmount, sliderMax=1}
  n3 = TextureSampleParameter2D() {parameterName=Texture, texture=&/T_UI_LeagueDivision_BasicIcon1}
  n4 = ScalarParameter() {defaultValue=0.1627, parameterName=Edge}
  n5 = Add(A=@n4, B=ScalarParameter() {defaultValue=0.1, parameterName=EdgeSoftness})
  n6 = MaterialFunctionCall(Alpha=@n3.RGB, Min=@n4, Max=@n5) {materialFunction=&SmoothStep}
  LinearInterpolate(
   A=LinearInterpolate(A=VectorParameter() {parameterName=OutlineColor, defaultValue=<rgba 0.0284 0.0003 0.117 1>}.RGB, B=@n1.RGB, Alpha=OneMinus(None=@n2))
   B=@n1.RGB
   Alpha=@n6.Result
  )
  n7 = Add(A=@n5, B=ScalarParameter() {defaultValue=-0.25, parameterName=Outline})
  n8 = Add(A=@n7, B=ScalarParameter() {defaultValue=0.1, parameterName=EdgeSoftness})
  Max(
   A=@n6.Result
   B=Multiply(
    A=MaterialFunctionCall(Alpha=@n3.RGB, Min=@n7, Max=@n8) {materialFunction=&SmoothStep}.Result
    B=ScalarParameter() {parameterName=OutlineAmount, sliderMax=1}
   )
  )
 ]
}
