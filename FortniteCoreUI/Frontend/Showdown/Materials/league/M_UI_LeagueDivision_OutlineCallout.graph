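# Translucent UI material for an outline callout on league division graphics in the Showdown frontend.
# It runs the red channel of the OutlineCallout texture through a SmoothStep to form a thresholded mask.
# The Edge parameter sets the lower bound and EdgeSoftness adds to it for the upper bound, so cutoff and softness adjust independently, similar to distance-field edge rendering.
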Material /FortniteCoreUI/Frontend/Showdown/Materials/league/M_UI_LeagueDivision_OutlineCallout
graph {
 properties {
  bUsedWithStaticMesh=false
  blendMode=BLEND_Translucent
  materialDomain=MD_UI
  pixelDepthOffsetMode=PDOM_Legacy
  usedShadingModels=Default Lit
 }
 nodes [
  Constant() {r=1}
  n1 = ScalarParameter() {defaultValue=0.2, parameterName=Edge}
  MaterialFunctionCall(
   Alpha=TextureSample() {texture=&/T_UI_LeagueDivision_OutlineCallout}.R
   Min=@n1
   Max=Add(A=@n1, B=ScalarParameter() {defaultValue=0.01, parameterName=EdgeSoftness})
  ) {materialFunction=&SmoothStep}
 ]
}
